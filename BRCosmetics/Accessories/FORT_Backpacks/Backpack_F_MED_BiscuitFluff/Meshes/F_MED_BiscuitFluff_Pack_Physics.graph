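# Physics asset for the BiscuitFluff backpack cosmetic, giving its dynamic ear and bun bones bodies and joints for secondary motion.
# A base capsule hangs freely off the pack root, carrying two-segment capsule ear chains with narrow limits and clustered bun spheres.
# Mid and upper buns pair free joints to a mid hub with tight constraints to the base.
# Lower buns join the base directly with tight swing and free twist.
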PhysicsAsset /BRCosmetics/Accessories/FORT_Backpacks/Backpack_F_MED_BiscuitFluff/Meshes/F_MED_BiscuitFluff_Pack_Physics
bodies {
 dyn_base_pack [{shapeType=Capsule, length=9.0741, radius=13.6938, center=<1.0293 -6.9617 -10.5527>, rotation=<rot 0.6591 -90.2657 1.5038>}]
 dyn_buns_base_01_l [{shapeType=Sphere, radius=3.1445, shapeName=dyn_buns_base_01_l_sphere, center=<1.831 -0.1291 0.2215>}]
 dyn_buns_base_01_m [{shapeType=Sphere, radius=3.3571, shapeName=dyn_buns_base_01_m_sphere, center=<2.1819 -0.02 0.0663>}]
 dyn_buns_base_01_r [{shapeType=Sphere, radius=3.0266, shapeName=dyn_buns_base_01_r_sphere, center=<2.1932 0.1428 0.0627>}]
 dyn_buns_mid_base [{shapeType=Sphere, radius=0.505, shapeName=dyn_buns_mid_base_sphere}]
 dyn_buns_mid_base_l [{shapeType=Sphere, radius=3.1334, shapeName=dyn_buns_mid_base_l_sphere, center=<1.5654 -0.2594 -0.0708>}]
 dyn_buns_mid_base_m [{shapeType=Sphere, radius=3.2702, shapeName=dyn_buns_mid_base_m_sphere, center=<2.0095 0.073 0.3782>}]
 dyn_buns_mid_base_r [{shapeType=Sphere, radius=3.1815, shapeName=dyn_buns_mid_base_r_sphere, center=<2.0195 -0.4233 0.7363>}]
 dyn_buns_mid_upper_l [{shapeType=Sphere, radius=3.1982, shapeName=dyn_buns_mid_upper_l_sphere, center=<1.4793 0.0135 -0.5302>}]
 dyn_buns_mid_upper_m [{shapeType=Sphere, radius=3.2797, shapeName=dyn_buns_mid_upper_m_sphere, center=<1.6753 0.3816 -0.3398>}]
 dyn_buns_mid_upper_r [{shapeType=Sphere, radius=3.2096, shapeName=dyn_buns_mid_upper_r_sphere, center=<-0.1924 -0.0037 1.7902>}]
 dyn_ear_base_l [{
   shapeType=Capsule
   length=2.2215
   radius=3.0895
   shapeName=dyn_ear_base_l_capsule
   center=<3.4976 0.0574 -0.4695>
   rotation=<rot -0.7224 91.3181 -89.872>
  }]
 dyn_ear_base_r [{
   shapeType=Capsule
   length=1.6302
   radius=3.225
   shapeName=dyn_ear_base_r_capsule
   center=<-3.2595 0.3004 0.449>
   rotation=<rot 0.0248 -91.0836 -90.0005>
  }]
 dyn_ear_l [{
   shapeType=Capsule
   length=2.3062
   radius=3.132
   shapeName=dyn_ear_l_capsule
   center=<3.1237 0.289 -0.5221>
   rotation=<rot 0.453 -86.7618 -89.9209>
  }]
 dyn_ear_r [{
   shapeType=Capsule
   length=2.8021
   radius=3.128
   shapeName=dyn_ear_r_capsule
   center=<-2.6146 0.0938 0.5809>
   rotation=<rot -2.1468 -87.6042 -86.2777>
  }]
 pack [{shapeType=Sphere, radius=0.505, shapeName=pack_sphere}]
}
constraints [
 {
  bone1Name=dyn_base_pack
  bone2Name=pack
  swing1LimitDegrees=45
  swing1Motion=Free
  swing2LimitDegrees=45
  swing2Motion=Free
  twistLimitDegrees=45
  twistMotion=Free
 }
 {
  bone1Name=dyn_buns_mid_base
  bone2Name=dyn_base_pack
  swing1LimitDegrees=45
  swing1Motion=Free
  swing2LimitDegrees=45
  swing2Motion=Free
  twistLimitDegrees=45
  twistMotion=Free
 }
 {
  bone1Name=dyn_buns_mid_base_l
  bone2Name=dyn_buns_mid_base
  swing1LimitDegrees=45
  swing1Motion=Free
  swing2LimitDegrees=45
  swing2Motion=Free
  twistLimitDegrees=45
  twistMotion=Free
 }
 {
  bone1Name=dyn_buns_mid_base_m
  bone2Name=dyn_buns_mid_base
  swing1LimitDegrees=45
  swing1Motion=Free
  swing2LimitDegrees=45
  swing2Motion=Free
  twistLimitDegrees=45
  twistMotion=Free
 }
 {
  bone1Name=dyn_buns_mid_base_r
  bone2Name=dyn_buns_mid_base
  swing1LimitDegrees=45
  swing1Motion=Free
  swing2LimitDegrees=45
  swing2Motion=Free
  twistLimitDegrees=45
  twistMotion=Free
 }
 {
  bone1Name=dyn_buns_mid_upper_l
  bone2Name=dyn_buns_mid_base
  swing1LimitDegrees=45
  swing1Motion=Free
  swing2LimitDegrees=45
  swing2Motion=Free
  twistLimitDegrees=45
  twistMotion=Free
 }
 {
  bone1Name=dyn_buns_mid_upper_m
  bone2Name=dyn_buns_mid_base
  swing1LimitDegrees=45
  swing1Motion=Free
  swing2LimitDegrees=45
  swing2Motion=Free
  twistLimitDegrees=45
  twistMotion=Free
 }
 {
  bone1Name=dyn_buns_mid_upper_r
  bone2Name=dyn_buns_mid_base
  swing1LimitDegrees=45
  swing1Motion=Free
  swing2LimitDegrees=45
  swing2Motion=Free
  twistLimitDegrees=45
  twistMotion=Free
 }
 {bone1Name=dyn_ear_base_l, bone2Name=dyn_base_pack, swing1LimitDegrees=15, swing2LimitDegrees=10, twistLimitDegrees=1}
 {bone1Name=dyn_ear_l, bone2Name=dyn_ear_base_l, swing1LimitDegrees=15, swing2LimitDegrees=10, twistLimitDegrees=1}
 {bone1Name=dyn_ear_base_r, bone2Name=dyn_base_pack, swing1LimitDegrees=15, swing2LimitDegrees=10, twistLimitDegrees=1}
 {bone1Name=dyn_ear_r, bone2Name=dyn_ear_base_r, swing1LimitDegrees=15, swing2LimitDegrees=10, twistLimitDegrees=1}
 {bone1Name=dyn_buns_mid_base_l, bone2Name=dyn_base_pack, swing1LimitDegrees=5, swing2LimitDegrees=5, twistLimitDegrees=1}
 {bone1Name=dyn_buns_mid_base_m, bone2Name=dyn_base_pack, swing1LimitDegrees=5, swing2LimitDegrees=5, twistLimitDegrees=1}
 {bone1Name=dyn_buns_mid_base_r, bone2Name=dyn_base_pack, swing1LimitDegrees=5, swing2LimitDegrees=5, twistLimitDegrees=1}
 {bone1Name=dyn_buns_mid_upper_l, bone2Name=dyn_base_pack, swing1LimitDegrees=5, swing2LimitDegrees=5, twistLimitDegrees=1}
 {bone1Name=dyn_buns_mid_upper_m, bone2Name=dyn_base_pack, swing1LimitDegrees=5, swing2LimitDegrees=5, twistLimitDegrees=1}
 {bone1Name=dyn_buns_mid_upper_r, bone2Name=dyn_base_pack, swing1LimitDegrees=5, swing2LimitDegrees=5, twistLimitDegrees=1}
 {bone1Name=dyn_buns_base_01_l, bone2Name=dyn_base_pack, swing1LimitDegrees=3, swing2LimitDegrees=2, twistLimitDegrees=45, twistMotion=Free}
 {bone1Name=dyn_buns_base_01_m, bone2Name=dyn_base_pack, swing1LimitDegrees=3, swing2LimitDegrees=2, twistLimitDegrees=45, twistMotion=Free}
 {bone1Name=dyn_buns_base_01_r, bone2Name=dyn_base_pack, swing1LimitDegrees=3, swing2LimitDegrees=2, twistLimitDegrees=45, twistMotion=Free}
]
